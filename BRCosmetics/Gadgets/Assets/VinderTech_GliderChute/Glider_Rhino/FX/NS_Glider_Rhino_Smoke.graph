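# Niagara smoke trail effect for the jets on the Rhino glider cosmetic, built from one sprite emitter.
# Despite its LEFT1 name, the emitter samples four left and right jet sockets on the skeletal mesh, choosing among them with a converted-Cascade SequentialInt input.
# Short-lived puffs launch along Y, grow over life, flash warm orange before settling to dusky tint, then fade.
# SubUV animation and curve-driven dynamic material parameters shape each puff.
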NiagaraSystem /BRCosmetics/Gadgets/Assets/VinderTech_GliderChute/Glider_Rhino/FX/NS_Glider_Rhino_Smoke
dependencies {
 usedDataInterfaces [
  &NiagaraDataInterfaceSkeletalMesh
  &NiagaraDataInterfaceCurve
  &NiagaraDataInterfaceSpriteRendererInfo
  &NiagaraDataInterfaceVectorCurve
 ]
 usedDynamicInputs []
 usedModules [&SystemState]
}
emitters {LEFT1_Smoke {
  emitterUpdateScript [
   &EmitterState(Life Cycle Mode=::NewEnumerator1, Loop Duration=1, UseLoopDelay=true)
   &SpawnRate(SpawnRate=30, Use Spawn Probability=false)
  ]
  particleSpawnScript [
   &InitializeParticle(
    Color=<rgba 0.2031 0.0678 0.0624 0.38>
    Lifetime=&RandomRangeFloat(Minimum=0.35, Maximum=0.48, Evaluation Type=::NewEnumerator1)
    Sprite Rotation Angle=&RandomRangeFloat(Minimum=0, Maximum=1, Evaluation Type=::NewEnumerator1)
    Sprite Rotation Mode=::NewEnumerator1
    Sprite Size=&Vector2DFromFloat(Value=&RandomRangeFloat(Minimum=42, Maximum=60, Evaluation Type=::NewEnumerator1))
    Sprite Size Mode=::NewEnumerator4
    UsePositionOffset=false
    Write Material Random=true
   )
   &AddVelocity(Velocity=&RandomRangeVector(Minimum=<0 320 0>, Maximum=<0 400 0>, Evaluation Type=::NewEnumerator1))
   &SkeletalMeshLocation(
    Bone / Socket Index=/CascadeToNiagaraConverter/NiagaraScripts/DynamicInputScripts/SequentialInt(Use Array Data:NiagaraBool=false, First Number:NiagaraInt32=0, Sequence:NiagaraDataInterfaceArrayInt32=?, Last Number:NiagaraInt32=3)
    MaskSpawnGroup=false
    Mesh Sampling Type=::NewEnumerator5
    Orientation Sampling=::NewEnumerator1
    Skeletal Mesh=DataInterface() {propertyValues {
      DefaultMesh=null
      ExcludeBoneName=root
      OptionalComponentClass=null
      PreviewMesh=/BRCosmetics/Gadgets/Assets/VinderTech_GliderChute/Glider_Rhino/Meshes/Glider_Rhino
      SkinningMode=SkinOnTheFly
      SoftSourceActor=null
      SourceMode=Default
      UvSetIndex=0
      WholeMeshLOD=-1
      bExcludeBone=false
      bReadDeformedGeometry=true
      bRequireCurrentFrameData=true
      ComponentTags []
      FilteredBones []
      FilteredSockets [jetB_jnt_LSocket_1, jetB_jnt_LSocket_2, jetB_jnt_RSocket_1, jetB_jnt_RSocket_2]
      MeshUserParameter {parameter {name=None, typeDefHandle {registeredTypeIndex=1}, varData []}}
      SamplingRegions []
     }}
    Socket Sampling Mode=::NewEnumerator5
   )
  ]
  particleUpdateScript [
   &ParticleState()
   &ScaleSpriteSize(
    Scale Sprite Size Mode=::NewEnumerator0
    Uniform Scale Factor=&FloatFromCurve(FloatCurve=DataInterface() {propertyValues {Curve {keys [{arriveTangent=0.6122, interpMode=RCIM_Cubic, leaveTangent=0.6122, tangentMode=RCTM_User, time=0, value=0.4}, <at 1 2.5>]}}})
   )
   &SubUVAnimation(End Frame Range Override=2, SubUV Animation Mode:ENiagara_SubUVLookupModeV2=::NewEnumerator2, UseEndFrame=false, UseStartFrame=false)
   &DynamicMaterialParameters(
    Index 0 Param 1=&FloatFromCurve(FloatCurve=DataInterface() {propertyValues {Curve {keys [<at 0 0.5>,<at 1 1>]}}})
    Index 0 Param 2=0.02
    Index 0 Param 3=&FloatFromCurve(FloatCurve=DataInterface() {propertyValues {Curve {keys [<at 0 0.2>,<at 0.5 1>]}}})
    Index 0 Param 4=1
    Module.Param0WriteEnabled=true
    Module.Param1WriteEnabled=true
    Module.Param2WriteEnabled=true
    Module.Param3WriteEnabled=true
   )
   &ScaleColor(
    Scale Alpha=&FloatFromCurve(FloatCurve=DataInterface() {propertyValues {Curve {keys [
         <at 0.0019 0.7062 RCIM_Cubic>
         {arriveTangent=2.8664, interpMode=RCIM_Cubic, leaveTangent=2.8664, tangentMode=RCTM_User, time=0.0975, value=1.0088}
         {interpMode=RCIM_Cubic, leaveTangent=4.5113, tangentMode=RCTM_Break, time=0.2027, value=1.497}
         {arriveTangent=-4.2437, interpMode=RCIM_Cubic, leaveTangent=-4.2437, tangentMode=RCTM_User, time=0.536, value=1.3193}
         {arriveTangent=-2.8436, interpMode=RCIM_Cubic, leaveTangent=-2.8436, time=0.7547, value=0.1962}
         <at 1 0>
        ]}}})
    Scale RGB=&VectorFromCurve(VectorCurve=DataInterface() {propertyValues {
       XCurve {keys [<at 0 2 RCIM_Cubic>,<at 0.2 1 RCIM_Cubic>]}
       YCurve {keys [<at 0 0.3925 RCIM_Cubic>,<at 0.2 1 RCIM_Cubic>]}
       ZCurve {keys [<at 0 0.1731 RCIM_Cubic>,<at 0.2 1 RCIM_Cubic>]}
      }})
    ScaleA=true
    ScaleRGB=true
   )
   &ScaleVelocity(Velocity Scale=&VectorFromCurve(VectorCurve=DataInterface() {propertyValues {XCurve {keys [<at 0 0>,<at 0.7 0>]}, YCurve {keys [<at 0 0.5>,<at 0.7 0.8>]}, ZCurve {keys [<at 0 1>,<at 0.7 1>]}}}))
   &SolveForcesAndVelocity(Clamp Velocity=false, Limit Acceleration=false)
  ]
  rendererClasses [&NiagaraSpriteRendererProperties]
  renderers [{rendererClass=&NiagaraSpriteRendererProperties}]
 }}
summary {
 systemName=NS_Glider_Rhino_Smoke
 emitters [{emitterName=LEFT1_Smoke, rendererClasses [&NiagaraSpriteRendererProperties]}]
 userVariables []
}
